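; Base pickup actor that applies a configurable gameplay effect to a pawn that overlaps it, then destroys itself.
; On authority, the sphere overlap calls OnPickUp, which checks tag queries, required pawn tags, downed-state tags and optionally missing health.
; A successful pickup sets Picked Up, whose OnRep plays pickup particles and sound.
; Spawn particles play on BeginPlay where cosmetics can run, and despawn particles play if it is destroyed unpicked.
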(event EventBeginPlay
  ;[1 1 1] Event Begin Play
  (Networking|CanExecuteCosmeticEvents
    (:True
      (CallFunction|InitialSetUp))))

(event OnComponentBeginOverlap(Sphere)
  (params (OverlappedComponent :type Primitive%20Component%20Object%20Reference) (OtherActor :type Actor%20Object%20Reference) (OtherComp :type Primitive%20Component%20Object%20Reference) (OtherBodyIndex :type Integer) (bFromSweep :type Boolean) (SweepResult :type Hit%20Result%20Structure%20%28by%20ref%29))
  ;[1 1 1] On Begin Overlap
  (Networking|SwitchHasAuthority
    (:Authority
      (CallFunction|OnPickUp OtherActor))))

(event EventDestroyed
  ;[1 1 1] Event Destroyed
  (if (not (Variables|Default|GetPicked%20Up))
    (Effects|SpawnEmitteratLocation (Variables|Effects|GetP_Despawn) (+ (Transformation|GetActorLocation) (Variables|Default|GetPsysInitialLocationOffset)))))

(fn InitialSetUp
  (Effects|SpawnEmitteratLocation (Variables|Effects|GetP_OnSpawn) (+ (Transformation|GetActorLocation) (Variables|Default|GetPsysInitialLocationOffset))))

(fn OnPickUp
  (params (Picking%20Up%20Actor :type Actor%20Object%20Reference))
  (if (not (GameplayTags|IsTagQueryEmpty (Variables|Default|GetPawnTagQuery)))
    (bind _returnvalue (GameplayTags|GetOwnedGameplayTags Picking%20Up%20Actor :!class /Script/GameplayTags.GameplayTagAssetInterface))
    (if (GameplayTags|DoesContainerMatchTagQuery _returnvalue (Variables|Default|GetPawnTagQuery))
      _))
  (if (Math|Boolean|ANDBoolean (Utilities|IsValid Picking%20Up%20Actor :!class /Script/Engine.KismetSystemLibrary) (GameplayTags|TagContainer|CompareTagAssetInterfacetoOtherTagContainers (Variables|Default|GetPawn%20Tags) (Ability|GetAbilitySystemComponent Picking%20Up%20Actor)) (not (GameplayTags|HasAnyMatchingGameplayTags (Ability|GetAbilitySystemComponent Picking%20Up%20Actor) (Variables|Default|GetPawn%20DBNO))))
    (bind _asfort_player_pawn (Utilities|Casting|CastToFortPlayerPawn Picking%20Up%20Actor))
    (if (Variables|Default|GetCheckHealth)
      (if (< (Math|Float|Floor (FortPawn|GetHealth _asfort_player_pawn)) (Math|Float|Floor (FortPawn|GetMaxHealth _asfort_player_pawn)))
        _))
    (GameplayEffects|ApplyGameplayEffectToTarget :self (Ability|GetAbilitySystemComponent Picking%20Up%20Actor) :GameplayEffectClass (Variables|Default|GetGE%20to%20Apply%20On%20Pick%20Up) :Target (Ability|GetAbilitySystemComponent Picking%20Up%20Actor) :Level (Math|Conversions|ToFloat(Integer) (Variables|Default|GetIntGameplayEffectLevel)))
    (Variables|Default|SetPicked%20Up true)
    (Actor|DestroyActor)))

(fn OnRep_PickedUp
  (if (Variables|Default|GetPicked%20Up)
    (Effects|SpawnEmitteratLocation (Variables|Effects|GetP_PickUp) (+ (Transformation|GetActorLocation) (Variables|Default|GetPsysInitialLocationOffset)))
    (Audio|PlaySoundatLocation (Variables|Default|GetPick%20Up%20Sound%20Effect) (Transformation|GetWorldLocation (Variables|Collision|GetCapsuleComponent)))))

(fn ConstructionScript)
